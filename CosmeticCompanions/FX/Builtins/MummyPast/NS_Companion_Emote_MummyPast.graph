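# Niagara effect for the MummyPast companion emote, adding smoke puffs and saw blur to emote objects.
# Both emitters compare user object-index floats against cached values to trigger per-frame spawns, then attach to Object Reader skeletal mesh bones chosen by a parsed index.
# SawBlur spawns two oppositely rotated spinning radial meshes only when IsSaw is set.
# SmokeSpawn also tracks a concurrent index, tinting short-lived sprites by time of day and saw state.
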NiagaraSystem /CosmeticCompanions/FX/Builtins/MummyPast/NS_Companion_Emote_MummyPast
dependencies {
 usedDataInterfaces [&NiagaraDataInterfaceCurve, &NiagaraDataInterfaceSkeletalMesh]
 usedDynamicInputs [
  &Multiply_FloatByInt
  &MergeHelper_IntSwitch
  &MakeIntFromFloat
  &self:Merge_Helper_Float_Switch_15
 ]
 usedModules [
  &self:SystemSpawnScript.SystemScriptSource.SystemScriptGraph.NiagaraNodeAssignment_8.SetVariables_#1
  &SystemState
  &self:SystemSpawnScript.SystemScriptSource.SystemScriptGraph.NiagaraNodeAssignment_0.SetVariables_#2
  &self:SystemSpawnScript.SystemScriptSource.SystemScriptGraph.NiagaraNodeAssignment_1.SetVariables_#3
 ]
}
emitters {
 SawBlur {
  emitterUpdateScript [
   &EmitterState(Life Cycle Mode=::NewEnumerator1, Loop Behavior=::NewEnumerator1, Loop Duration Mode=::NewEnumerator1, UseLoopDelay=true)
   &SpawnPerFrame(
    Spawn:NiagaraBool=&SetBoolByFloatComparison(A=$User.ObjectIndex, B=$Emitter.ObjectIndexInitial)
    Spawn Count:NiagaraInt32=&Multiply_Int(A=2, B=$System.IsSaw)
    Use Spawn Probability=false
   )
   /SetVariables(Emitter.ObjectIndexInitial:NiagaraFloat=$User.ObjectIndex) {isSetParameters=true}
  ]
  particleSpawnScript [
   &InitializeParticle(
    Color=<rgba 0.06 0.06 0.06 0.8>
    Lifetime=1.6
    Mesh Scale Mode=::NewEnumerator3
    Mesh Uniform Scale=2.5
    Position Offset=<0 40 0>
    UsePositionOffset=true
    Write Material Random=true
   )
   &FN_SetSpawnGroup(Number of Spawn Groups=2)
   &InitialMeshOrientation(
    Mesh Axis To Orient=<0 1 0>
    Mesh Orientation Mode=::NewEnumerator2
    Orientation Vector=<0 0 1>
    Rotation=&MaskVectorBySpawnGroup(Masked Vector=<0 0 0.25>, Passthrough Vector=<0 0 -0.25>)
    Use Rotation Vector=true
   )
   /SetVariables(Particles.ObjectParserInitial:NiagaraInt32=$System.ObjectReaderParser) {isSetParameters=true}
  ]
  particleUpdateScript [
   &ParticleState()
   &ScaleMeshSize(Scale Factor=&VectorFromFloat(Value=&FloatFromCurve(FloatCurve=DataInterface() {propertyValues {Curve {keys [<at 0 0>,<at 0.1006 1>,<at 0.9111 1>,<at 1 0>]}}})))
   &SkeletalMeshLocation(
    Bone / Socket Index=$Particles.ObjectParserInitial
    Bone Sampling Mode=::NewEnumerator4
    MaskSpawnGroup=false
    Skeletal Mesh=$System.Object Reader
   )
   &FN_Radial_Material(Base Sprite Tile U=4, Base Sprite Tile V=0.1, Depth Fade=0, Radial Mask Power=1, Radial Mask Tile V=3, Sprite Spin Speed=15)
   &UpdateMeshOrientation(Rotation Vector=<0 0 1>)
   &CameraOffset(Camera Offset Amount=10)
   &ScaleColor(ScaleA=true, ScaleRGB=true)
  ]
  rendererClasses [&NiagaraMeshRendererProperties]
  renderers [{rendererClass=&NiagaraMeshRendererProperties, rendererIndex=0}]
 }
 SmokeSpawn {
  emitterUpdateScript [
   &EmitterState()
   &SpawnPerFrame(
    Spawn:NiagaraBool=&SetBoolByFloatComparison(A=$User.ObjectIndex, B=$Emitter.ObjectIndexInitial)
    Spawn Count:NiagaraInt32=&Multiply_Int(A=1, B=1)
    Use Spawn Probability=false
   )
   &SpawnPerFrame(
    Spawn:NiagaraBool=&SetBoolByFloatComparison(A=$User.ConcurrentObjectIndex, B=$Emitter.ConcurrentObjIndInit)
    Spawn Count:NiagaraInt32=&Multiply_Int(A=1, B=1)
    Spawn Group=1
    Use Spawn Probability=false
   ) {module=SpawnPerFrame001}
   /SetVariables(Emitter.ConcurrentObjIndInit:NiagaraFloat=$User.ConcurrentObjectIndex, Emitter.ObjectIndexInitial:NiagaraFloat=$User.ObjectIndex) {isSetParameters=true}
  ]
  particleSpawnScript [
   /SetVariables(Particles.ObjectParserInitial:NiagaraInt32=$System.ObjectReaderParser) {isSetParameters=true}
   &InitializeParticle(
    Color=&FNTimeOfDayColor(
     Night Color=<rgba 0.41 0.4757 0.842 1>
     Morning Color=<rgba 0.7078 0.5363 0.5148 1>
     Day Color=<rgba 0.8096 0.8096 0.8096 1>
     Evening Color=<rgba 0.6076 0.4758 0.7031 1>
    )
    Lifetime Max=0.35
    Lifetime Min=0.3
    Lifetime Mode=::NewEnumerator1
    Sprite Rotation Angle=&RandomRangeFloat(Minimum=-45, Maximum=45)
    Sprite Rotation Mode=::NewEnumerator0
    Sprite Size Mode=::NewEnumerator1
    Uniform Sprite Size Max=80
    Uniform Sprite Size Min=60
    UsePositionOffset=false
    Write Material Random=true
   ) {module=InitializeParticle001}
   &CameraOffset(Camera Offset Amount=20)
   &SkeletalMeshLocation(
    Bone / Socket Index=$Particles.ObjectParserInitial
    Bone Sampling Mode=::NewEnumerator4
    MaskSpawnGroup=false
    Skeletal Mesh=$System.Object Reader
   )
   &DynamicMaterialParameters(
    Index 0 Param 2=&UniformAOrBFloat(A=0)
    Module.Param0WriteEnabled=true
    Module.Param1WriteEnabled=true
    Module.Param2WriteEnabled=true
    Module.Param3WriteEnabled=true
   ) {module=DynamicMaterialParameters001}
  ]
  particleUpdateScript [
   &ParticleState()
   &SolveForcesAndVelocity(Clamp Velocity=false, Limit Acceleration=false)
   &DynamicMaterialParameters(
    Index 0 Param 1=&FloatFromCurve(FloatCurve=DataInterface() {propertyValues {Curve {keys [<at 0 0>,<at 1 1>]}}})
    Index 0 Param 2=1
    Module.Param0WriteEnabled=true
    Module.Param1WriteEnabled=true
    Module.Param2WriteEnabled=true
    Module.Param3WriteEnabled=true
   )
   &ScaleColor(
    Scale Alpha=&FloatFromCurve(FloatCurve=DataInterface() {propertyValues {Curve {keys [<at 0 1>,<at 0.8091 0.9932>,<at 1 1>]}}})
    Scale RGB=&MergeHelper_VectorSwitch(0=<1 1 1>, 1=<0.65 0.65 1>, 2=?, 3=?, Index=$System.IsSaw, Num=::NewEnumerator0)
    ScaleA=true
    ScaleRGB=true
   )
  ]
  rendererClasses [&NiagaraSpriteRendererProperties]
  renderers [{rendererClass=&NiagaraSpriteRendererProperties}]
 }
}
summary {
 systemName=NS_Companion_Emote_MummyPast
 emitters [
  {emitterName=SawBlur, rendererClasses [&NiagaraMeshRendererProperties]}
  {emitterName=SmokeSpawn, rendererClasses [&NiagaraSpriteRendererProperties]}
 ]
 userVariables [
  User.bIsSwinging {type=NiagaraBool}
  User.ConcurrentObjectIndex {type=NiagaraFloat, defaultValue=0, description=""}
  User.JetActiveCurve {type=NiagaraFloat, defaultValue=1, description=""}
  User.ObjectIndex {type=NiagaraFloat, defaultValue=0, description=""}
  User.RockProbability {type=NiagaraFloat, defaultValue=1, description=""}
 ]
}
