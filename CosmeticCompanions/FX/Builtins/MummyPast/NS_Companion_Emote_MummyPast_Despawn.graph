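# Niagara effect for the MummyPast cosmetic companion's despawn emote, playing a short smoke puff.
# A single SmokeSpawn emitter bursts fixed-size sprites with random rotation and a time-of-day color.
# Dynamic material parameters animate the puff over its brief life, with a curve ramping from zero to one.
# User.SawIndex sets a visibility tag across two sprite renderers and selects a time-of-day RGB scale.
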NiagaraSystem /CosmeticCompanions/FX/Builtins/MummyPast/NS_Companion_Emote_MummyPast_Despawn
dependencies {usedDataInterfaces [&NiagaraDataInterfaceCurve], usedDynamicInputs [], usedModules [&SystemState]}
emitters {SmokeSpawn {
  emitterUpdateScript [
   &EmitterState()
   &SpawnBurst_Instantaneous(Use Spawn Probability=false, UseLoopCountLimit=false)
  ]
  particleSpawnScript [
   &InitializeParticle(
    Color=&FNTimeOfDayColor(
     Night Color=<rgba 0.432 0.4943 0.842 1>
     Morning Color=<rgba 0.7078 0.5363 0.5148 1>
     Day Color=<rgba 0.8096 0.8096 0.8096 1>
     Evening Color=<rgba 0.6076 0.4758 0.7031 1>
    )
    Lifetime Max=0.34
    Lifetime Min=0.34
    Lifetime Mode=::NewEnumerator1
    Sprite Rotation Angle=&RandomRangeFloat(Minimum=-45, Maximum=45)
    Sprite Rotation Mode=::NewEnumerator0
    Sprite Size Mode=::NewEnumerator1
    Uniform Sprite Size Max=70
    Uniform Sprite Size Min=70
    UsePositionOffset=false
    Write Material Random=true
   ) {module=InitializeParticle001}
   &CameraOffset(Camera Offset Amount=20)
   &DynamicMaterialParameters(
    Index 0 Param 2=&UniformAOrBFloat(A=0)
    Module.Param0WriteEnabled=true
    Module.Param1WriteEnabled=true
    Module.Param2WriteEnabled=true
    Module.Param3WriteEnabled=true
   ) {module=DynamicMaterialParameters001}
   /SetVariables(Particles.VisibilityTag=&MergeHelper_IntSwitch(0=0, 1=1, 2=0, 3=?, 4=?, 5=?, Index=&MakeIntFromFloat(Float=$User.SawIndex), Num=::NewEnumerator1)) {isSetParameters=true}
   /Game/Sprite/FN_SpriteAlignmentAndFacing(SpriteAlignment:Vector3f=<0 0 1>, SpriteFacing:Vector3f=<0 0 1>)
  ]
  particleUpdateScript [
   &ParticleState()
   &SolveForcesAndVelocity(Clamp Velocity=false, Limit Acceleration=false)
   &DynamicMaterialParameters(
    Index 0 Param 1=&FloatFromCurve(FloatCurve=DataInterface() {propertyValues {Curve {keys [<at 0 0>,<at 1 1>]}}})
    Index 0 Param 2=1
    Module.Param0WriteEnabled=true
    Module.Param1WriteEnabled=true
    Module.Param2WriteEnabled=true
    Module.Param3WriteEnabled=true
   )
   &ScaleColor(
    Scale Alpha=&FloatFromCurve(FloatCurve=DataInterface() {propertyValues {Curve {keys [<at 0 1>,<at 0.8091 0.9932>,<at 1 1>]}}})
    Scale RGB=&MergeHelper_VectorSwitch(
     0=<1 1 1>
     1=&FN_Select_Vector_from_Time_Of_Day(Morning=<0.9 0.9 1>, Day=<0.65 0.65 1>, Evening=<0.9 0.9 1>, Night=<1 1 1>)
     2=&FN_Select_Vector_from_Time_Of_Day(Morning=<0.9 0.9 1>, Day=<0.651 0.65 1>, Evening=<0.9 0.9 1>, Night=<1 1 1>)
     3=?
     Index=&MakeIntFromFloat(Float=$User.SawIndex)
     Num=::NewEnumerator1
    )
    ScaleA=true
    ScaleRGB=true
   )
  ]
  rendererClasses [&NiagaraSpriteRendererProperties]
  renderers [
   {rendererClass=&NiagaraSpriteRendererProperties}
   {rendererClass=&NiagaraSpriteRendererProperties, rendererIndex=1}
  ]
 }}
summary {
 systemName=NS_Companion_Emote_MummyPast_Despawn
 emitters [{emitterName=SmokeSpawn, rendererClasses [&NiagaraSpriteRendererProperties]}]
 userVariables [
  User.bIsSwinging {type=NiagaraBool}
  User.ConcurrentObjectIndex {type=NiagaraFloat, defaultValue=0, description=""}
  User.JetActiveCurve {type=NiagaraFloat, defaultValue=1, description=""}
  User.ObjectIndex {type=NiagaraFloat, defaultValue=0, description=""}
  User.RockProbability {type=NiagaraFloat, defaultValue=1, description=""}
  User.SawIndex {type=NiagaraFloat, defaultValue=0, description=""}
 ]
}
