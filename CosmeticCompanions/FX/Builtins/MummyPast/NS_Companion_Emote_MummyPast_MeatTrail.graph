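# Niagara system for the MummyPast companion emote's meat trail, drawing short-lived red ribbons.
# Two near-identical emitters spawn ribbon particles at a steady rate; only the first samples a system-level skeletal mesh reader for spawn data.
# Both scale color with a fade-in/fade-out alpha curve, multiplied by a time-of-day value that dims at night.
# User floats expose saw, double-saw and object selection, though their wiring is not shown.
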NiagaraSystem /CosmeticCompanions/FX/Builtins/MummyPast/NS_Companion_Emote_MummyPast_MeatTrail
dependencies {
 usedDataInterfaces [&NiagaraDataInterfaceCurve, &NiagaraDataInterfaceSkeletalMesh]
 usedDynamicInputs [
  &MergeHelper_FloatSwitch
  &MakeIntFromFloat
  &MergeHelper_VectorSwitch
 ]
 usedModules [
  &SystemState
  &self:SystemSpawnScript.SystemScriptSource.SystemScriptGraph.NiagaraNodeAssignment_0.SetVariables_#1
 ]
}
emitters {
 Minimal {
  emitterUpdateScript [
   &EmitterState()
   &SpawnRate(SpawnRate=80, Use Spawn Probability=false)
  ]
  particleSpawnScript [
   &SampleSkeletalMesh(
    AdditionalYawPitchRoll=false
    Bone Sampling Mode:ENiagaraBoneSamplingMode=::NewEnumerator4
    Skeletal Mesh:NiagaraDataInterfaceSkeletalMesh=$System.Skeletal Mesh Reader
   )
   &InitializeParticle(
    Color=<rgba 1 0.038 0.024 1>
    Lifetime=0.15
    Ribbon Width=41
    Ribbon Width Mode=::NewEnumerator1
    UsePositionOffset=false
    Write Material Random=true
   )
  ]
  particleUpdateScript [
   &ParticleState()
   &ScaleColor(
    Scale Alpha=&FloatFromCurve(
     FloatCurve=DataInterface() {propertyValues {Curve {
        defaultValue=1
        keys [
         {interpMode=RCIM_Cubic, tangentMode=RCTM_User, time=0, value=0}
         {interpMode=RCIM_Cubic, tangentMode=RCTM_User, time=0.1, value=1}
         {tangentMode=RCTM_User, time=1, value=0}
        ]
       }}}
     Scale Curve=&FN_Select_Float_from_Time_Of_Day(Morning=1.5, Day=2, Evening=1.5, Night=0.7)
    )
    ScaleA=true
    ScaleRGB=true
   )
  ]
  rendererClasses [&NiagaraRibbonRendererProperties]
  renderers [{rendererClass=&NiagaraRibbonRendererProperties}]
 }
 Minimal001 {
  emitterUpdateScript [
   &EmitterState()
   &SpawnRate(SpawnRate=80, Use Spawn Probability=false)
  ]
  particleSpawnScript [&InitializeParticle(
    Color=<rgba 1 0.051 0.031 1>
    Lifetime=0.15
    Ribbon Width=41
    Ribbon Width Mode=::NewEnumerator1
    UsePositionOffset=false
    Write Material Random=true
   )]
  particleUpdateScript [
   &ParticleState()
   &ScaleColor(
    Scale Alpha=&FloatFromCurve(
     FloatCurve=DataInterface() {propertyValues {Curve {
        defaultValue=1
        keys [
         {interpMode=RCIM_Cubic, tangentMode=RCTM_User, time=0, value=0}
         {interpMode=RCIM_Cubic, tangentMode=RCTM_User, time=0.1, value=1}
         {tangentMode=RCTM_User, time=1, value=0}
        ]
       }}}
     Scale Curve=&FN_Select_Float_from_Time_Of_Day(Morning=1.5, Day=2, Evening=1.5, Night=0.7)
    )
    ScaleA=true
    ScaleRGB=true
   )
  ]
  rendererClasses [&NiagaraRibbonRendererProperties]
  renderers [{rendererClass=&NiagaraRibbonRendererProperties}]
 }
}
summary {
 systemName=NS_Companion_Emote_MummyPast_MeatTrail
 emitters [
  {emitterName=Minimal, rendererClasses [&NiagaraRibbonRendererProperties]}
  {emitterName=Minimal001, rendererClasses [&NiagaraRibbonRendererProperties]}
 ]
 userVariables [
  User.DoubleSawIndex {type=NiagaraFloat, defaultValue=0, description=""}
  User.IsDoubleSaw {type=NiagaraFloat, defaultValue=0, description=""}
  User.ObjectIndex {type=NiagaraFloat, defaultValue=0, description=""}
  User.SawIndex {type=NiagaraFloat, defaultValue=0, description=""}
 ]
}
