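# Niagara saw trail effect for the MummyPast companion emote, drawing swing trails and a spinning blur on skeletal mesh sockets or bones.
# Two ribbon emitters face along the sampled bone rotation, while Propeller2 bursts radial-material meshes that follow the bone, tinted by time of day.
# User.SawIndex selects bone, timing and size through switch helpers.
# User.IsDoubleSaw stops trail spawning and retimes the blur, with DoubleSawIndex choosing a side offset for saw index 2.
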NiagaraSystem /CosmeticCompanions/FX/Builtins/MummyPast/NS_Companion_Emote_MummyPast_SawTrail
dependencies {
 usedDataInterfaces [&NiagaraDataInterfaceCurve, &NiagaraDataInterfaceSkeletalMesh]
 usedDynamicInputs []
 usedModules [
  &SystemState
  &self:SystemSpawnScript.SystemScriptSource.SystemScriptGraph.NiagaraNodeAssignment_0.SetVariables_#1
 ]
}
emitters {
 Minimal {
  emitterUpdateScript [
   &EmitterState(
    Life Cycle Mode=::NewEnumerator1
    Loop Behavior=::NewEnumerator1
    Loop Duration=&MergeHelper_FloatSwitch(0=0.75, 1=0.35, 2=0.35, 3=1.6, Index=&MakeIntFromFloat(Float=$User.SawIndex), Num=::NewEnumerator2)
    UseLoopDelay=false
   )
   &SpawnRate(
    SpawnRate=&FloatFromCurve(
     FloatCurve=DataInterface() {propertyValues {Curve {keys [<at 0.8 1>,<at 1 0>]}}}
     CurveIndex=$Emitter.NormalizedLoopAge
     Scale Curve=&MergeHelper_FloatSwitch(0=80, 1=0, 2=?, 3=?, Index=&MakeIntFromFloat(Float=$User.IsDoubleSaw), Num=::NewEnumerator0)
    )
    Use Spawn Probability=false
   )
  ]
  particleSpawnScript [
   &InitializeParticle(
    Color=&MakeLinearColor(
     R:NiagaraFloat=1
     G:NiagaraFloat=1
     B:NiagaraFloat=1
     A:NiagaraFloat=&MergeHelper_FloatSwitch(
      0=&FloatFromCurve(FloatCurve=DataInterface() {propertyValues {Curve {keys [<at 0.5 1>,<at 0.8 0>]}}}, CurveIndex=$Emitter.NormalizedLoopAge)
      1=&FloatFromCurve(FloatCurve=DataInterface() {propertyValues {Curve {keys [<at 0.3 1>,<at 0.5 0>]}}}, CurveIndex=$Emitter.NormalizedLoopAge)
      2=&FloatFromCurve(FloatCurve=DataInterface() {propertyValues {Curve {keys [<at 0.3 1>,<at 0.5 0>]}}}, CurveIndex=$Emitter.NormalizedLoopAge)
      3=&FloatFromCurve(FloatCurve=DataInterface() {propertyValues {Curve {keys [<at 1 1>,<at 1.5 0>]}}}, CurveIndex=$Emitter.NormalizedLoopAge)
      Index=&MakeIntFromFloat(Float=$User.SawIndex)
      Num=::NewEnumerator2
     )
    )
    Lifetime=&Multiply_Float(
     A=&FloatFromCurve(
      FloatCurve=DataInterface() {propertyValues {Curve {defaultValue=1, keys [{interpMode=RCIM_Cubic, tangentMode=RCTM_User, time=0.3, value=1}, {tangentMode=RCTM_User, time=1, value=0}]}}}
      CurveIndex=$Emitter.NormalizedLoopAge
      Scale Curve=&MergeHelper_FloatSwitch(0=0.3, 1=0.1, 2=0.1, 3=0, Index=&MakeIntFromFloat(Float=$User.SawIndex), Num=::NewEnumerator2)
     )
     B=1
    )
    Ribbon Facing Mode=::NewEnumerator1
    Ribbon Width=&Multiply_Float(
     A=&MergeHelper_FloatSwitch(0=35, 1=21, 2=?, 3=?, Index=&MakeIntFromFloat(Float=$User.SawIndex), Num=::NewEnumerator0)
     B=&FloatFromCurve(FloatCurve=DataInterface() {propertyValues {Curve {keys [<at 0.6 1>,<at 0.8 0.5>]}}}, CurveIndex=$Emitter.NormalizedLoopAge)
    )
    Ribbon Width Mode=::NewEnumerator1
    UsePositionOffset=false
    Write Material Random=true
   )
   &SkeletalMeshLocation(
    Bone / Socket Index=&MakeIntFromFloat(Float=$User.SawIndex)
    Bone Sampling Mode=::NewEnumerator4
    MaskSpawnGroup=false
    Orientation Sampling=::NewEnumerator1
    Sampled Position Offset=&MakeVector(X=0, Y=0, Z=0)
    Skeletal Mesh=$System.Skeletal Mesh Reader
   )
   /SetVariables(Particles.RibbonFacing=&RotateVectorByQuaternion(Vector To Rotate=<0 1 0>, Quaternion=$Particles.SkeletalMeshLocation.SampledRotation)) {isSetParameters=true}
  ]
  particleUpdateScript [
   &ParticleState()
   &ScaleColor(
    Scale Alpha=&FloatFromCurve(
     FloatCurve=DataInterface() {propertyValues {Curve {
        defaultValue=1
        keys [
         {interpMode=RCIM_Cubic, tangentMode=RCTM_User, time=0.1275, value=0}
         {interpMode=RCIM_Cubic, tangentMode=RCTM_User, time=0.2218, value=1}
         {tangentMode=RCTM_User, time=1, value=0}
        ]
       }}}
     Scale Curve=1.4
    )
    ScaleA=true
    ScaleRGB=true
   )
  ]
  rendererClasses [&NiagaraRibbonRendererProperties, &NiagaraMeshRendererProperties]
  renderers [
   {rendererClass=&NiagaraRibbonRendererProperties}
   {rendererClass=&NiagaraMeshRendererProperties, rendererIndex=1}
  ]
 }
 Minimal001 {
  emitterUpdateScript [
   &EmitterState(
    Life Cycle Mode=::NewEnumerator1
    Loop Behavior=::NewEnumerator1
    Loop Duration=&MergeHelper_FloatSwitch(0=0.75, 1=0.35, 2=0.35, 3=1.6, Index=&MakeIntFromFloat(Float=$User.SawIndex), Num=::NewEnumerator2)
    UseLoopDelay=false
   )
   &SpawnRate(
    SpawnRate=&Multiply_Float(
     A=&MergeHelper_FloatSwitch(
      0=&FloatFromCurve(FloatCurve=DataInterface() {propertyValues {Curve {keys [<at 0.8 1>,<at 1 0>]}}}, CurveIndex=$Emitter.NormalizedLoopAge, Scale Curve=80)
      1=&FloatFromCurve(FloatCurve=DataInterface() {propertyValues {Curve {keys [<at 0.4 1>,<at 0.6 0>]}}}, CurveIndex=$Emitter.NormalizedLoopAge, Scale Curve=80)
      2=?
      3=?
      Index=&MakeIntFromFloat(Float=$User.SawIndex)
      Num=::NewEnumerator0
     )
     B=&MergeHelper_FloatSwitch(0=80, 1=0, 2=?, 3=?, Index=&MakeIntFromFloat(Float=$User.IsDoubleSaw), Num=::NewEnumerator0)
    )
    Use Spawn Probability=false
   )
  ]
  particleSpawnScript [
   &InitializeParticle(
    Color=&MakeLinearColor(
     R:NiagaraFloat=1
     G:NiagaraFloat=1
     B:NiagaraFloat=1
     A:NiagaraFloat=&MergeHelper_FloatSwitch(
      0=&FloatFromCurve(FloatCurve=DataInterface() {propertyValues {Curve {keys [<at 0.5 1>,<at 0.8 0>]}}}, CurveIndex=$Emitter.NormalizedLoopAge, Scale Curve=0.4)
      1=&FloatFromCurve(FloatCurve=DataInterface() {propertyValues {Curve {keys [<at 0.3 1>,<at 0.5 0>]}}}, CurveIndex=$Emitter.NormalizedLoopAge, Scale Curve=0.4)
      2=&FloatFromCurve(FloatCurve=DataInterface() {propertyValues {Curve {keys [<at 0.3 1>,<at 0.5 0>]}}}, CurveIndex=$Emitter.NormalizedLoopAge, Scale Curve=0.4)
      3=0
      Index=&MakeIntFromFloat(Float=$User.SawIndex)
      Num=::NewEnumerator2
     )
    )
    Lifetime=&Multiply_Float(
     A=&FloatFromCurve(
      FloatCurve=DataInterface() {propertyValues {Curve {defaultValue=1, keys [{interpMode=RCIM_Cubic, tangentMode=RCTM_User, time=0.3, value=1}, {tangentMode=RCTM_User, time=1, value=0}]}}}
      CurveIndex=$Emitter.NormalizedLoopAge
      Scale Curve=&MergeHelper_FloatSwitch(0=0.3, 1=0.1, 2=0.1, 3=0, Index=&MakeIntFromFloat(Float=$User.SawIndex), Num=::NewEnumerator2)
     )
     B=1
    )
    Ribbon Facing Mode=::NewEnumerator1
    Ribbon Facing Vector:Vector3f=&MergeHelper_VectorSwitch(0=<1 1.5 0>, 1=<0 0 1>, 2=?, 3=?, Index=&MakeIntFromFloat(Float=$User.SawIndex), Num=::NewEnumerator0)
    Ribbon Width=&Multiply_Float(
     A=&MergeHelper_FloatSwitch(0=37, 1=25, 2=?, 3=?, Index=&MakeIntFromFloat(Float=$User.SawIndex), Num=::NewEnumerator0)
     B=&MergeHelper_FloatSwitch(
      0=&FloatFromCurve(FloatCurve=DataInterface() {propertyValues {Curve {keys [<at 0.6 1>,<at 0.8 0.5>]}}}, CurveIndex=$Emitter.NormalizedLoopAge)
      1=&FloatFromCurve(FloatCurve=DataInterface() {propertyValues {Curve {keys [<at 0.45 1>,<at 0.6 0.5>]}}}, CurveIndex=$Emitter.NormalizedLoopAge)
      2=?
      3=?
      Index=&MakeIntFromFloat(Float=$User.SawIndex)
      Num=::NewEnumerator0
     )
    )
    Ribbon Width Mode=::NewEnumerator1
    UsePositionOffset=false
    Write Material Random=true
   )
   &SkeletalMeshLocation(
    Bone / Socket Index=&MakeIntFromFloat(Float=$User.SawIndex)
    Bone Sampling Mode=::NewEnumerator4
    MaskSpawnGroup=false
    Orientation Sampling=::NewEnumerator1
    Skeletal Mesh=$System.Skeletal Mesh Reader
   )
   /SetVariables(Particles.RibbonFacing=&RotateVectorByQuaternion(Vector To Rotate=<0 1 0>, Quaternion=$Particles.SkeletalMeshLocation.SampledRotation)) {isSetParameters=true}
  ]
  particleUpdateScript [
   &ParticleState()
   &ScaleColor(
    Scale Alpha=&FloatFromCurve(FloatCurve=DataInterface() {propertyValues {Curve {
        defaultValue=1
        keys [
         {interpMode=RCIM_Cubic, tangentMode=RCTM_User, time=0.0402, value=0}
         {interpMode=RCIM_Cubic, tangentMode=RCTM_User, time=0.146, value=1}
         {tangentMode=RCTM_User, time=1, value=0}
        ]
       }}})
    ScaleA=true
    ScaleRGB=true
   )
  ]
  rendererClasses [&NiagaraRibbonRendererProperties]
  renderers [{rendererClass=&NiagaraRibbonRendererProperties}]
 }
 Propeller2 {
  emitterUpdateScript [
   &EmitterState(
    Inactive Response=::NewEnumerator1
    Life Cycle Mode=::NewEnumerator1
    Loop Behavior=::NewEnumerator1
    Loop Duration=&MergeHelper_FloatSwitch(0=0.9, 1=0.35, 2=1.5, 3=1.61, Index=&MakeIntFromFloat(Float=$User.SawIndex), Num=::NewEnumerator2)
    UseLoopDelay=true
   )
   &SpawnBurst_Instantaneous(Spawn Count=4, Use Spawn Probability=false, UseLoopCountLimit=false)
  ]
  particleSpawnScript [
   &PartitionParticles()
   &InitializeParticle(
    Color=&MergeHelper_ColorSwitch(
     0=<rgba 0.06 0.06 0.06 0.8>
     1=<rgba 0.3906 0.3906 0.3906 0.3826>
     2=?
     3=?
     Index=$Particles.PartitionParticles.Partition
     Num=::NewEnumerator0
    )
    Lifetime=&MergeHelper_FloatSwitch(
     0=&MergeHelper_FloatSwitch(0=0.8, 1=0, 2=0, 3=1.61, Index=&MakeIntFromFloat(Float=$User.SawIndex), Num=::NewEnumerator2)
     1=1.9
     2=?
     3=?
     Index=&MakeIntFromFloat(Float=$User.IsDoubleSaw)
     Num=::NewEnumerator0
    )
    Mesh Scale Mode=::NewEnumerator3
    Mesh Uniform Scale=&MergeHelper_FloatSwitch(0=1.5, 1=0.9, 2=?, 3=?, Index=&MakeIntFromFloat(Float=$User.SawIndex), Num=::NewEnumerator0)
    Position Offset=<0 40 0>
    UsePositionOffset=true
    Write Material Random=true
   )
   &FN_SetSpawnGroup(Number of Spawn Groups=2)
   &InitialMeshOrientation(
    Mesh Axis To Orient=<0 1 0>
    Mesh Orientation Mode=::NewEnumerator2
    Orientation Vector=<0 0 1>
    Rotation=&MaskVectorBySpawnGroup(Masked Vector=<0 0 0.25>, Passthrough Vector=<0 0 -0.25>)
    Use Rotation Vector=true
   )
  ]
  particleUpdateScript [
   &ParticleState()
   &SkeletalMeshLocation(
    Bone / Socket Index=&MakeIntFromFloat(Float=$User.SawIndex)
    Bone Sampling Mode=::NewEnumerator4
    MaskSpawnGroup=false
    Sampled Position Offset=&MergeHelper_VectorSwitch(
     0=<0 0 0>
     1=<0 0 0>
     2=&MergeHelper_VectorSwitch(
      0=<0 36 0>
      1=<0 -36 0>
      2=?
      3=?
      Index=&MakeIntFromFloat(Float=$User.DoubleSawIndex)
      Num=::NewEnumerator0
     )
     3=?
     Index=&MakeIntFromFloat(Float=$User.SawIndex)
     Num=::NewEnumerator1
    )
    Skeletal Mesh=$System.Skeletal Mesh Reader
   )
   &ScaleMeshSize(Scale Factor=&Multiply_Vector(
     A=$Particles.Initial.Scale
     B=&MergeHelper_VectorSwitch(
      0=&MergeHelper_VectorSwitch(
       0=&VectorFromFloat(Value=&FloatFromCurve(
         FloatCurve=DataInterface() {propertyValues {Curve {keys [<at 0 0>,<at 0.1288 1>,<at 0.5 1>,<at 0.9434 0>]}}}
         CurveIndex=$Emitter.NormalizedLoopAge
        ))
       1=<0 0 0>
       2=<0 0 0>
       3=&VectorFromFloat(Value=&FloatFromCurve(
         FloatCurve=DataInterface() {propertyValues {Curve {keys [<at 0 0>,<at 0.057 1>,<at 0.8778 1>,<at 1 0>]}}}
         CurveIndex=$Emitter.NormalizedLoopAge
        ))
       Index=&MakeIntFromFloat(Float=$User.SawIndex)
       Num=::NewEnumerator2
      )
      1=&VectorFromFloat(Value=&FloatFromCurve(
        FloatCurve=DataInterface() {propertyValues {Curve {keys [<at 0 0>,<at 0.07 1>,<at 0.95 1>,<at 1 0>]}}}
        CurveIndex=$Emitter.NormalizedLoopAge
       ))
      2=?
      3=?
      Index=&MakeIntFromFloat(Float=$User.IsDoubleSaw)
      Num=::NewEnumerator0
     )
    ))
   &FN_Radial_Material(Base Sprite Tile U=4, Base Sprite Tile V=0.1, Depth Fade=0, Radial Mask Power=1, Radial Mask Tile V=3, Sprite Spin Speed=15)
   &UpdateMeshOrientation(Rotation Vector=<0 0 1>)
   &CameraOffset(Camera Offset Amount=6)
   &ScaleColor(
    Scale RGB=&FN_Select_Vector_from_Time_Of_Day(Morning=<1 1 1>, Day=<1 1 1>, Evening=<1 1 1>, Night=<0.5 0.5 1>)
    ScaleA=true
    ScaleRGB=true
   )
  ]
  rendererClasses [&NiagaraMeshRendererProperties]
  renderers [{rendererClass=&NiagaraMeshRendererProperties, rendererIndex=0}]
 }
}
summary {
 systemName=NS_Companion_Emote_MummyPast_SawTrail
 emitters [
  {emitterName=Minimal, rendererClasses [&NiagaraRibbonRendererProperties, &NiagaraMeshRendererProperties]}
  {emitterName=Propeller2, rendererClasses [&NiagaraMeshRendererProperties]}
  {emitterName=Minimal001, rendererClasses [&NiagaraRibbonRendererProperties]}
 ]
 userVariables [
  User.DoubleSawIndex {type=NiagaraFloat, defaultValue=0, description=""}
  User.IsDoubleSaw {type=NiagaraFloat, defaultValue=0, description=""}
  User.ObjectIndex {type=NiagaraFloat, defaultValue=0, description=""}
  User.SawIndex {type=NiagaraFloat, defaultValue=0, description=""}
 ]
}
